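# Animation Blueprint for the Urban Scavenger backpack cosmetic, adding physics-driven secondary motion to its knife, straps and buckles.
# From the reference pose, eight chained Anim Dynamics nodes simulate those bones using damped angular springs, reduced gravity and angular limits, some in root-relative space.
# A final Modify Bone additively rotates backpack_root by backpack_rotate, which nothing here sets, as the event graph holds only unconnected stub nodes.
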AnimBlueprint /BRCosmetics/Accessories/FORT_Backpacks/Backpack_M_MED_Urban_Scavenger/Meshes/M_MED_Urban_Scavenger_Pack_AnimBP
target_skeleton=&/M_MED_Urban_Scavenger_Pack_Skeleton
graphs [
 AnimGraph {nodes [
   n1 = Anim Dynamics - Bone: dyn_knifebuckle_01_r(ComponentPose=Local To Component(LocalPose=Local Space Ref Pose()), Alpha=2, ExternalForce="0,0,0") {
    angularDampingOverride=1
    angularSpringConstant=500
    bAngularSpring=true
    bGravityOverrideInSimSpace=true
    bOverrideAngularDamping=true
    bUsePlanarLimit=false
    gravityScale=0.35
    simulationSpace=RootRelative
    alphaBoolBlend {blendOption=Linear}
    boundBone {boneName=dyn_knifebuckle_01_r}
    componentAppliedLinearAccClamp=<50 50 50>
    componentLinearAccScale=<0.5 0.5 0.5>
    componentLinearVelScale=<0.5 0.5 0.5>
    physicsBodyDefinitions [{
      sphereCollisionRadius=0
      boundBone {boneName=dyn_knifebuckle_01_r}
      boxExtents=<5 5 5>
      constraintSetup {angularTargetAxis=AxisY, angularLimitsMax=<0 5 0>, angularLimitsMin=<0 -30 0>, angularTarget=<0 0 0>}
      localJointOffset=<1 0 0>
     }]
    relativeSpaceBone {boneName=backpack_root}
   }
   n2 = Anim Dynamics - Bone: dyn_knife_01_r(ComponentPose=@n1, Alpha=2, ExternalForce="0,0,0") {
    angularDampingOverride=1
    angularSpringConstant=500
    bAngularSpring=true
    bGravityOverrideInSimSpace=true
    bOverrideAngularDamping=true
    gravityScale=0.35
    alphaBoolBlend {blendOption=Linear}
    boundBone {boneName=dyn_knife_01_r}
    componentAppliedLinearAccClamp=<50 50 50>
    componentLinearAccScale=<0.5 0.5 0.5>
    componentLinearVelScale=<0.5 0.5 0.5>
    physicsBodyDefinitions [{
      sphereCollisionRadius=0
      boundBone {boneName=dyn_knife_01_r}
      boxExtents=<5 5 5>
      constraintSetup {angularTargetAxis=AxisY, angularLimitsMax=<0 5 0>, angularLimitsMin=<0 -15 0>, angularTarget=<0 0 0>}
      localJointOffset=<1 0 0>
     }]
   }
   n3 = Anim Dynamics - Bone: dyn_knife_03_r(ComponentPose=@n2, Alpha=0.5, ExternalForce="0,0,0") {
    angularBiasOverride=0.9999
    angularDampingOverride=1
    angularSpringConstant=500
    bAngularSpring=true
    bGravityOverrideInSimSpace=true
    bOverrideAngularDamping=true
    bUsePlanarLimit=false
    gravityScale=0.35
    numSolverIterationsPostUpdate=2
    numSolverIterationsPreUpdate=8
    alphaBoolBlend {blendOption=Linear}
    boundBone {boneName=dyn_knife_03_r}
    componentAppliedLinearAccClamp=<25 25 25>
    componentLinearAccScale=<0.3 0.3 0.3>
    componentLinearVelScale=<0.3 0.3 0.3>
    physicsBodyDefinitions [{
      sphereCollisionRadius=0
      boundBone {boneName=dyn_knife_03_r}
      boxExtents=<5 5 5>
      constraintSetup {angularLimitsMax=<5 25 50>, angularLimitsMin=<-5 -25 -2>, angularTarget=<0 0 0>}
      localJointOffset=<4 0 0>
     }]
    planarLimits [{drivingBone {boneName=backpack_root}, planeTransform {location=<11 0 0>, rotation=<rot -79.9993 -179.9999 -179.9999>}}]
   }
   n4 = Anim Dynamics - Bone: dyn_backstrap_01_r(ComponentPose=@n3, Alpha=2, ExternalForce="0,0,0") {
    angularDampingOverride=1
    angularSpringConstant=500
    bAngularSpring=true
    bGravityOverrideInSimSpace=true
    bOverrideAngularDamping=true
    gravityScale=0.35
    alphaBoolBlend {blendOption=Linear}
    boundBone {boneName=dyn_backstrap_01_r}
    componentAppliedLinearAccClamp=<250 250 250>
    componentLinearAccScale=<0.2 0.2 0.2>
    componentLinearVelScale=<0.2 0.2 0.2>
    physicsBodyDefinitions [{
      sphereCollisionRadius=0
      boundBone {boneName=dyn_backstrap_01_r}
      boxExtents=<2 2 2>
      constraintSetup {angularTargetAxis=AxisY, angularLimitsMax=<0 30 0>, angularLimitsMin=<0 -5 0>, angularTarget=<0 0 0>}
      localJointOffset=<-1.5 0 0>
     }]
   }
   n5 = Anim Dynamics - Bone: dyn_backstrap_01_l(ComponentPose=@n4, Alpha=2, ExternalForce="0,0,0") {
    angularDampingOverride=1
    angularSpringConstant=500
    bAngularSpring=true
    bGravityOverrideInSimSpace=true
    bOverrideAngularDamping=true
    gravityScale=0.35
    alphaBoolBlend {blendOption=Linear}
    boundBone {boneName=dyn_backstrap_01_l}
    componentAppliedLinearAccClamp=<250 250 250>
    componentLinearAccScale=<0.2 0.2 0.2>
    componentLinearVelScale=<0.2 0.2 0.2>
    physicsBodyDefinitions [{
      sphereCollisionRadius=0
      boundBone {boneName=dyn_backstrap_01_l}
      boxExtents=<2 2 2>
      constraintSetup {angularTargetAxis=AxisY, angularLimitsMax=<0 30 0>, angularLimitsMin=<0 -5 0>, angularTarget=<0 0 0>}
      localJointOffset=<-1.5 0 0>
     }]
   }
   n6 = Anim Dynamics - Bone: dyn_upperstrap_01_l(ComponentPose=@n5, Alpha=1, ExternalForce="0,0,0") {
    angularDampingOverride=1
    angularSpringConstant=500
    bAngularSpring=true
    bGravityOverrideInSimSpace=true
    bOverrideAngularDamping=true
    gravityScale=0.35
    simulationSpace=RootRelative
    alphaBoolBlend {blendOption=Linear}
    boundBone {boneName=dyn_upperstrap_01_l}
    componentAppliedLinearAccClamp=<250 250 250>
    componentLinearAccScale=<0.5 0.5 0.5>
    componentLinearVelScale=<0.5 0.5 0.5>
    physicsBodyDefinitions [{
      sphereCollisionRadius=0
      boundBone {boneName=dyn_upperstrap_01_l}
      boxExtents=<5 2 4>
      constraintSetup {angularTargetAxis=AxisY, angularLimitsMax=<0 0 30>, angularLimitsMin=<0 0 -5>, angularTarget=<0 0 0>}
      localJointOffset=<-1.5 0 0>
     }]
   }
   n7 = Anim Dynamics - Bone: dyn_lowerstrap_01_l(ComponentPose=@n6, Alpha=1, ExternalForce="0,0,0") {
    angularDampingOverride=1
    angularSpringConstant=500
    bAngularSpring=true
    bGravityOverrideInSimSpace=true
    bOverrideAngularDamping=true
    gravityScale=0.35
    simulationSpace=RootRelative
    alphaBoolBlend {blendOption=Linear}
    boundBone {boneName=dyn_lowerstrap_01_l}
    componentAppliedLinearAccClamp=<250 250 250>
    componentLinearAccScale=<0.5 0.5 0.5>
    componentLinearVelScale=<0.5 0.5 0.5>
    physicsBodyDefinitions [{
      sphereCollisionRadius=0
      boundBone {boneName=dyn_lowerstrap_01_l}
      boxExtents=<5 2 4>
      constraintSetup {angularTargetAxis=AxisY, angularLimitsMax=<0 0 35>, angularLimitsMin=<0 0 -5>, angularTarget=<0 0 0>}
      localJointOffset=<1.5 0 0>
     }]
   }
   n8 = Anim Dynamics - Bone: dyn_sidebuckle_01_l(ComponentPose=@n7, Alpha=2, ExternalForce="0,0,0") {
    angularDampingOverride=1
    angularSpringConstant=500
    bAngularSpring=true
    bGravityOverrideInSimSpace=true
    bOverrideAngularDamping=true
    gravityScale=0.35
    simulationSpace=RootRelative
    alphaBoolBlend {blendOption=Linear}
    boundBone {boneName=dyn_sidebuckle_01_l}
    componentAppliedLinearAccClamp=<250 250 250>
    componentLinearAccScale=<0.5 0.5 0.5>
    componentLinearVelScale=<0.5 0.5 0.5>
    physicsBodyDefinitions [{
      sphereCollisionRadius=0
      boundBone {boneName=dyn_sidebuckle_01_l}
      boxExtents=<5 8 10>
      constraintSetup {angularTargetAxis=AxisY, angularLimitsMax=<5 5 10>, angularLimitsMin=<-5 -5 -10>, angularTarget=<0 0 0>}
      localJointOffset=<-1.5 0 0>
     }]
   }
   n9 = "Transform (Modify) Bone - Bone: backpack_root"(ComponentPose=@n8, Alpha=1, Translation="0,0,0", Rotation=Get backpack_rotate(), Scale="1,1,1") {rotationMode=BMM_Additive, rotationSpace=BCS_BoneSpace, alphaBoolBlend {blendOption=Linear}, boneToModify {boneName=backpack_root}}
   Output Pose(Result=Component To Local(ComponentPose=@n9)) {layerGroup=None}
  ]}
 EventGraph {nodes [
   TryGetPawnOwner()
   Event BlueprintUpdateAnimation()
  ]}
]
variables [backpack_rotate {type=Rotator}]
